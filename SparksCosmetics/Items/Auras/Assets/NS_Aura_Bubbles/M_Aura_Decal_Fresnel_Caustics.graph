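# Unlit translucent Niagara sprite material for the Bubbles aura, producing a decal-like caustic glow on scene surfaces.
# A panning caustic texture, distorted by a scrolling cloud-noise normal map, forms the mask, while squared particle color drives emissive.
# Opacity math combines particle alpha, a radial sphere mask, depth fades and a Fresnel from flattened scene-buffer normals, tuned by dynamic parameters.
# Low quality and mobile replace the Fresnel with a constant vector parameter.
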Material /SparksCosmetics/Items/Auras/Assets/NS_Aura_Bubbles/M_Aura_Decal_Fresnel_Caustics
graph {
 properties {bUsedWithNiagaraSprites=true, blendMode=BLEND_Translucent, pixelDepthOffsetMode=PDOM_Legacy, shadingModel=MSM_Unlit, usedShadingModels=Unlit}
 nodes [
  n1 = ParticleColor()
  MaterialFunctionCall(Emissive=Multiply(A=ScalarParameter() {defaultValue=5, parameterName=EmissiveMult}, B=Multiply(A=@n1.RGB, B=@n1.RGB))) {materialFunction=&MF_FX_EmissiveAdjustments}
  n2 = MaterialFunctionCall(Float3=Divide(A=WorldPosition().XYZ, B=ScalarParameter() {defaultValue=2500, parameterName=World Position Scale})) {materialFunction=&BreakOutFloat3Components}
  n3 = OneMinus(None=@n2.B)
  NamedRerouteDeclaration(None=AppendVector(A=@n3, B=@n2.G) {desc=X}) {name=X, nodeColor=<rgba 0.8281 0 1 1>}
  NamedRerouteDeclaration(None=AppendVector(A=@n2.R, B=@n3) {desc=Y}) {name=Y, nodeColor=<rgba 0.9453 0 1 1>}
  NamedRerouteDeclaration(None=AppendVector(A=@n2.R, B=@n2.G) {desc=Z}) {name=Z, nodeColor=<rgba 0 1 0.6719 1>}
  n4 = Panner(
   Coordinate=Multiply(A=TextureCoordinate(), B=ScalarParameter() {defaultValue=0.5, parameterName=CausticDistortionScale})
   Speed=AppendVector(
    A=ScalarParameter() {defaultValue=0.25, parameterName=Caustic Distortion Speed X}
    B=ScalarParameter() {defaultValue=0.25, parameterName=Caustic Distortion Speed Y}
   )
  )
  n5 = Multiply(
   A=ComponentMask(None=TextureSample(UVs=@n4) {samplerType=SAMPLERTYPE_Normal, texture=&/T_Aura_Bubbles_CloudNoise_N}.RGB) {g=true, r=true}
   B=ScalarParameter() {defaultValue=0.1, parameterName=Caustic Distortion Strength}
  )
  n6 = Multiply(A=TextureCoordinate())
  n7 = ScalarParameter() {defaultValue=-0.175, parameterName=Caustic Speed X}
  n8 = ScalarParameter() {defaultValue=0.175, parameterName=Caustic Speed Y}
  n9 = TextureSample(UVs=Panner(Coordinate=Add(A=@n5, B=@n6), Speed=AppendVector(A=@n7, B=@n8))) {samplerType=SAMPLERTYPE_LinearColor, texture=/FortWater/Textures/T_Caustic_M}
  n10 = Multiply(A=@n9.R)
  n11 = Multiply(A=@n9.G) {constB=0.5}
  NamedRerouteDeclaration(None=Multiply(A=Add(A=@n10, B=@n11), B=Constant() {r=0.85})) {name=CausticMask, nodeColor=<rgba 1 0 0.3047 1>}
  NamedRerouteUsage()
  NamedRerouteUsage()
  ParticlePositionWS()
  n12 = NamedRerouteUsage()
  n13 = DynamicParameter() {defaultValue=<rgba 3 0.15 2 1>, paramNames [FresnelExponent, FlattenNormal, Param3, Param4]}
  n14 = SceneTexture() {sceneTextureId=PPI_WorldNormal}
  n15 = Fresnel(
   ExponentIn=@n13.FresnelExponent
   Normal=MaterialFunctionCall(Normal=ComponentMask(None=@n14.Color) {b=true, g=true, r=true}, Flatness=@n13.FlattenNormal) {materialFunction=&FlattenNormal}.Result
  ) {baseReflectFraction=0, exponent=1}
  n16 = VectorParameter() {parameterName=Forward, defaultValue=<rgba 0 0 1 1>}
  n17 = Multiply(
   A=ShadingPathSwitch(Default=QualitySwitch(Default=@n15, Low=@n16.RGB, High=@n15, Medium=@n15, Epic=@n15), Mobile=@n16.RGB)
   B=SphereMask(
    A=TextureCoordinate()
    B=Constant() {r=0.5}
    Radius=ScalarParameter() {defaultValue=0.5, parameterName=SphereMaskRadius}
    Hardness=ScalarParameter() {parameterName=SphereMaskHardness}
   ) {hardnessPercent=0}
  )
  n18 = ScalarParameter() {defaultValue=15, parameterName=VisibleDepthDistance}
  n19 = DepthFade(
   Opacity=OneMinus(None=DepthFade(FadeDistance=@n18) {fadeDistanceDefault=80})
   FadeDistance=ScalarParameter() {defaultValue=15, parameterName=DepthFade}
  ) {fadeDistanceDefault=15}
  Reroute(None=Multiply(A=Multiply(A=@n12, B=@n1.A), B=Multiply(A=@n17, B=@n19)))
 ]
}
