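# Translucent, unlit, two-sided bubble material for Niagara mesh particles in the Bubbles aura cosmetic.
# Color layers combine a cosine-palette iridescent gradient, rotated per particle over time, with sky-light highlights, tinted cubemap reflections and Fresnel glow.
# Fresnel-style terms also shape index-of-refraction distortion and opacity, both scaled by particle alpha.
# Switchable curl-style vector noise, masked by vertex red, drives vertex wobble, and dynamic parameters control a noise-texture erosion mask.
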Material /SparksCosmetics/Items/Auras/Assets/NS_Aura_Bubbles/M_Aura_TranslucentBubble
graph {
 properties {
  bOutputTranslucentVelocity=true
  bUsedWithNiagaraMeshParticles=true
  blendMode=BLEND_Translucent
  pixelDepthOffsetMode=PDOM_Legacy
  refractionMethod=RM_IndexOfRefraction
  shadingModel=MSM_Unlit
  translucencyLightingMode=TLM_Surface
  twoSided=true
  usedShadingModels=Unlit
 }
 nodes [
  n1 = ParticleSize()
  AppendVector(A=@n1, B=ComponentMask(None=@n1) {r=true})
  Constant3Vector() {constant=<rgba 0.5 1 0.43 1>}
  Constant3Vector() {constant=<rgba 0 0 1 0>}
  Constant() {r=0.5}
  n2 = NamedRerouteUsage()
  n3 = Add(A=NamedRerouteUsage(), B=NamedRerouteUsage())
  n4 = NamedRerouteUsage()
  n5 = Constant() {r=0.5}
  n6 = Multiply(A=ParticleColor().RGB, B=Add(A=Add(A=@n2, B=@n3), B=Multiply(A=@n4, B=@n5)))
  MaterialFunctionCall(Emissive=@n6) {materialFunction=&MF_FX_EmissiveAdjustments}
  MaterialFunctionCall() {desc=Gradient for system bounds not what we want, materialFunction=&BoundingBoxBased_0-1_UVW}
  MaterialFunctionCall!(Center, Height) {materialFunction=/Game/Effects/Fort_Effects/Materials/_Functions/MF_Z_GradientFromObjectHeight_01}
  Multiply(A=NamedRerouteUsage(), B=NamedRerouteUsage())
  n7 = DotProduct(A=PixelNormalWS(), B=CameraVectorWS())
  NamedRerouteDeclaration(None=OneMinus(None=Abs(None=@n7))) {name=FresnelOriginValue, nodeColor=<rgba 0 0.0156 1 1>}
  n8 = Constant3Vector() {constant=<rgba 1 0 0 0>}
  n9 = Constant3Vector() {constant=<rgba 0 0 1 0>}
  n10 = Constant3Vector() {constant=<rgba 0 1 0 0>}
  n11 = DynamicParameter() {parameterIndex=3, defaultValue=<rgba 0 1 1 1>, paramNames [Timer,Scale,BounceStr,Random]}
  n12 = Add(A=@n11.Random, B=Time())
  n13 = ScalarParameter() {group=GradientColor, parameterName=GradientRotateAxisLerpSpeed}
  n14 = ConstantBiasScale(None=Cosine(None=Multiply(A=@n12, B=@n13)))
  n15 = ScalarParameter() {group=GradientColor, parameterName=GradientRotateSpeed}
  n16 = WorldPosition()
  n17 = ParticlePositionWS()
  n18 = RotateAboutAxis(
   NormalizedRotationAxis=Normalize(VectorInput=MaterialFunctionCall(1=@n8, 2=@n9, 3=@n10, A=@n14) {materialFunction=&Lerp_Multiple_Float3}.Lerp 3 Inputs)
   RotationAngle=Sine(None=Multiply(A=@n12, B=@n15))
   PivotPoint=Constant3Vector()
   Position=Divide(A=Subtract(A=@n16.XYZ, B=@n17), B=@n11.Scale)
  )
  n19 = ScalarParameter() {group=GradientColor, parameterName=GradientRotateTileScale}
  n20 = LinearInterpolate(
   A=DotProduct(A=CameraVectorWS(), B=PixelNormalWS())
   B=MaterialFunctionCall(Float3=Multiply(A=@n18, B=@n19)) {materialFunction=&BreakOutFloat3Components}.B
   Alpha=ScalarParameter() {group=GradientColor, parameterName=GradientAxisChangeLerp}
  )
  n21 = Add(
   A=Multiply(A=Time(), B=Constant() {r=-0.3})
   B=DynamicParameter() {parameterIndex=3, defaultValue=<rgba 0 1 1 1>, paramNames [Timer,Scale,BounceStr,Random]}.Random
  )
  n22 = Multiply(
   A=VectorParameter() {group=GradientColor, parameterName=Palette_C, defaultValue=<rgba 1 1 1 1>}.RGB
   B=Multiply(A=Add(A=@n20, B=@n21), B=ScalarParameter() {group=GradientColor, parameterName=GradientTotalTileScale}) {constB=2}
  )
  n23 = VectorParameter() {group=GradientColor, parameterName=Palette_D, defaultValue=<rgba 0 0.33 0.67 1>}
  n24 = Multiply(
   A=VectorParameter() {group=GradientColor, parameterName=Palette_B, defaultValue=<rgba 0.5 0.5 0.5 1>}.RGB
   B=Cosine(None=Add(A=@n22, B=@n23.RGB))
  )
  n25 = VectorParameter() {group=GradientColor, parameterName=Palette_A, defaultValue=<rgba 0.5 0.5 0.5 1>}
  n26 = Multiply(
   A=Desaturation(None=Add(A=@n25.RGB, B=@n24), Fraction=ScalarParameter() {group=GradientColor, parameterName=GradientDesaturation})
   B=ScalarParameter() {group=GradientColor, parameterName=GradientColorScale}
  )
  n27 = ScalarParameter() {group=GradientColor, parameterName=GradientColorAdd}
  NamedRerouteDeclaration(None=Saturate(None=Add(A=@n26, B=@n27))) {name=GradientColorOriginValue, nodeColor=<rgba 1 0.3984 0 1>}
  n28 = SkyAtmosphereLightDirection()
  n29 = DotProduct(A=Normalize(VectorInput=ComponentMask(None=@n28) {b=true, g=true, r=true}), B=PixelNormalWS())
  n30 = Multiply(A=Add(A=Abs(None=@n29))) {constB=0.5}
  n31 = Add(A=@n30, B=Constant() {r=0.005})
  n32 = Saturate(None=Power(Base=Saturate(None=@n31), Exp=Constant() {r=50}))
  n33 = ScalarParameter() {group=HighLight, parameterName=Highlight_OuterBrightness}
  n34 = ScalarParameter() {group=HighLight, parameterName=Highlight_CenterBrightness}
  n35 = Saturate(None=Add(A=@n30, B=Constant() {r=0.0002}))
  n36 = Constant() {r=10000}
  n37 = MaterialFunctionCall(In=Saturate(None=Power(Base=@n35, Exp=@n36)), Contrast=Constant() {r=1}) {materialFunction=&CheapContrast}
  n38 = Add(A=Saturate(None=Multiply(A=@n32, B=@n33)), B=Saturate(None=Multiply(A=@n34, B=@n37.Result)))
  n39 = VectorParameter() {group=HighLight, parameterName=Highlight_Color, defaultValue=<rgba 0 0 0 1>}
  n40 = Power(Base=NamedRerouteUsage(), Exp=Constant() {r=1.5})
  NamedRerouteDeclaration(None=Multiply(A=Multiply(A=@n38, B=@n39.RGB), B=Saturate(None=@n40))) {name=HighLightValue, nodeColor=<rgba 0 0.7891 1 1>}
  n41 = NamedRerouteUsage()
  n42 = ScalarParameter() {group=CubeMap, parameterName=CubeMapFresnelPower}
  n43 = Saturate(None=Add(A=Power(Base=@n41, Exp=@n42), B=ScalarParameter() {group=CubeMap, parameterName=CubeMapFresnelBase}))
  n44 = TextureSampleParameterCube!(UVs=ReflectionVectorWS(), Apply View MipBias) {group=CubeMap, parameterName=CubeMapTexture}
  n45 = ScalarParameter() {defaultValue=1, group=CubeMap, parameterName=CubeMapColorPower}
  n46 = Multiply(
   A=Saturate(None=Power(Base=@n44.RGB, Exp=@n45))
   B=VectorParameter() {group=CubeMap, parameterName=CubeMapColorMultiply, defaultValue=<rgba 0 0 0 1>}.RGB
  )
  NamedRerouteDeclaration(None=Multiply(A=NamedRerouteUsage(), B=Multiply(A=@n43, B=@n46))) {name=CubeMapValue, nodeColor=<rgba 0.5234 0 1 1>}
  n47 = NamedRerouteUsage()
  n48 = Saturate(None=Power(Base=@n47, Exp=ScalarParameter() {group=FresnelEmissive, parameterName=EmissiveFresnelPower}))
  n49 = ScalarParameter() {group=FresnelEmissive, parameterName=Emissive2FresnelPower}
  n50 = Saturate(None=Multiply(A=Power(Base=@n47, Exp=@n49), B=ScalarParameter() {group=FresnelEmissive, parameterName=Emissive2ColorMultiply}))
  NamedRerouteDeclaration(None=Multiply(A=VectorParameter() {group=FresnelEmissive, parameterName=EmissiveColor, defaultValue=<rgba 0 0 0 1>}.RGB, B=Add(A=@n48, B=@n50))) {name=FresnelEmissiveValue, nodeColor=<rgba 0 0.7187 1 1>}
  n51 = ScalarParameter() {defaultValue=1, group=Opacity, parameterName=OpacityScale}
  n52 = VertexNormalWS()
  n53 = Constant() {r=0.35}
  n54 = Power(Base=Abs(None=@n52), Exp=@n53) {constExponent=1.5}
  n55 = DotProduct(A=@n54, B=Constant3Vector() {constant=<rgba 1 1 1 0>})
  n56 = Multiply(
   A=MaterialFunctionCall(Color=Divide(A=@n54, B=@n55)) {materialFunction=&SplitComponents}.B
   B=DotProduct(A=Power(Base=@n52, Exp=@n53) {constExponent=5}, B=Constant3Vector() {constant=<rgba 0 0 1 0>})
  )
  n57 = Saturate(None=Add(A=Multiply(A=@n56) {constB=3}) {constB=0.1})
  n58 = Add(
   A=Power(Base=NamedRerouteUsage(), Exp=ScalarParameter() {group=Opacity, parameterName=OpacityExponentIn})
   B=ScalarParameter() {group=Opacity, parameterName=OpacityBaseReflectFractionIn}
  )
  n59 = Multiply(A=Constant() {r=0.6}, B=Multiply(A=Saturate(None=@n58), B=@n57))
  NamedRerouteDeclaration(None=Multiply(A=Multiply(A=@n59, B=@n51), B=ParticleColor().A)) {name=OpacityValue, nodeColor=<rgba 0.2891 0 1 1>}
  n60 = MaterialFunctionCall!(In, Mask, isVehicle, isMech) {materialFunction=/Game/Characters/Player/Male/Medium/Bodies/M_MED_Mastermind/Materials/MF_TwistedGoo}
  n61 = DynamicParameter() {defaultValue=<rgba 0.2 25 1.5 -0.1>, paramNames ["Curl Speed (2)", "Curl Scale (10)", "Curl Intensity (3)", Erosion]}
  n62 = StaticSwitchParameter(
   True=MaterialFunctionCall() {materialFunction=&LocalPosition}.Local Position
   False=WorldPosition().XYZ
  ) {parameterName=UseLocalPosition?}
  n63 = Multiply(A=@n61."Curl Speed (2)", B=Time())
  n64 = VectorNoise(World Position=Divide(A=Subtract(A=@n62, B=@n63), B=@n61."Curl Scale (10)")) {noiseFunction=VNF_VectorALU}
  n65 = Multiply(
   A=VertexColor().R
   B=ShadingPathSwitch(Default=Multiply(A=@n61."Curl Intensity (3)", B=@n64), Mobile=Constant() {r=1}) {desc=Migrated from FeatureLevelSwitch}
  )
  n66 = Constant()
  NamedRerouteDeclaration(None=Add(
    A=GetMaterialAttributes(None=@n60.Result) {attributeGetTypes [#1]}.World Position Offset
    B=StaticSwitchParameter(True=@n65, False=@n66) {defaultValue=true, parameterName=USE_WPO}
   )) {name=WPO, nodeColor=<rgba 1 0.4688 0 1>}
  n67 = ScalarParameter() {defaultValue=1, group=Refraction, parameterName=RefractionCenterStrength}
  n68 = ScalarParameter() {defaultValue=1, group=Refraction, parameterName=RefractionOuterStrength}
  n69 = Power(Base=NamedRerouteUsage(), Exp=ScalarParameter() {group=Refraction, parameterName=RefractionFresnelPower})
  n70 = ParticleColor()
  NamedRerouteDeclaration(None=LinearInterpolate(B=LinearInterpolate(A=@n67, B=@n68, Alpha=@n69), Alpha=Saturate(None=@n70.A)) {constA=1}) {name=RefractionValue, nodeColor=<rgba 0 1 0.0625 1>}
  n71 = Multiply(A=TextureCoordinate(), B=Constant2Vector() {g=0.5, r=1.5})
  n72 = ComponentMask(None=TextureCoordinate()) {g=true}
  n73 = DynamicParameter() {defaultValue=<rgba 0.2 25 1.5 -0.1>, paramNames ["Curl Speed (2)", "Curl Scale (10)", "Curl Intensity (3)", Erosion]}
  n74 = MaterialFunctionCall!(
   "Erosion Texture (RGBA)"=Multiply(A=TextureSample(UVs=@n71) {samplerType=SAMPLERTYPE_LinearColor, texture=&HighContrastNoise}.RGBA, B=@n72)
   Erosion Contrast=Constant() {r=0.9}
   Erosion Amount=@n73.Erosion
  ) {materialFunction=/Game/Effects/Niagara/MasterMaterials/Functions/Mojo_MF/MF_Erosion_Complex_Base}
  NamedRerouteDeclaration(None=Add(A=@n74.Color, B=Multiply(A=@n74.Erosion Edge Mask) {constB=2})) {name=Erosion Mask, nodeColor=<rgba 1 0 0.6797 1>}
  NamedRerouteUsage()
  NamedRerouteUsage()
  ObjectPositionWS() {desc=Previously used}
  OneMinus(None=ComponentMask(None=TextureCoordinate()) {r=true})
  OneMinus(None=@n72)
  ParticlePositionWS()
  n75 = Transform(None=WorldPosition().XYZ) {transformSourceType=TRANSFORMSOURCE_World, transformType=TRANSFORM_Local}
  n76 = Transform(None=ParticlePositionWS()) {transformSourceType=TRANSFORMSOURCE_World, transformType=TRANSFORM_Local}
  n77 = Multiply(A=ObjectBounds()) {constB=0.01, desc="Scale down To rough particle mesh size (?)"}
  n78 = Constant3Vector() {constant=<rgba 0 0 1 0>}
  n79 = ComponentMask(None=Divide(A=Subtract(A=@n75, B=@n76), B=Multiply(A=@n77, B=@n78))) {b=true}
  Power(Base=Divide(A=Add(A=@n79)))
  Reroute(None=@n57)
  Reroute(None=@n6)
  Sine(None=Multiply(A=Time()) {constB=0.5})
 ]
}
